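# MetaSound Source preset for the one-shot stereo sound that plays when a vehicle's Retrostrips boost starts.
# It inherits its graph from MSS_Vehicle_Boost_Start, supplying one start wave, an empty layered one-shot array, and settings for pitch range, speed-based pitch and one-shot timing.
# Boost attenuation, start and end concurrency sets, and modulated volume, high-pass and low-pass shape playback on the exterior boost sound class.
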MetaSoundSource /VehicleAudio/Boosts/Boost_Retrostrips/MSS_Vehicle_Boost_Retrostrips_Start_Preset
preset_parent=/VehicleAudio/Boosts/_ParentMetasound/MSS_Vehicle_Boost_Start
inputs [
 Boost.Pitch.Max {type=Float, access=Constructor}
 Boost.Pitch.Min {type=Float, access=Constructor}
 Boost.PitchByNormalSpeed {type=Bool, access=Constructor}
 Boost.Start.Oneshots.DelayTime {type=Time, access=Constructor}
 Boost.Start.Oneshots.NumRepeats {type=Int32, access=Constructor}
 Boost.Start.Oneshots.RepeatSpeed {type=Time, access=Constructor}
 Boost.Start.Oneshots.Volume {type=Float, access=Constructor}
 Boost.Start.Oneshots.WaveAsset {type=WaveAsset:Array, access=Constructor}
 Boost.Start.Volume {type=Float, access=Constructor}
 Boost.Start.WaveAsset {type=WaveAsset:Array, access=Constructor}
 Fort.Vehicle.Acceleration {type=Float}
 Fort.Vehicle.BlowoffStrength {type=Float}
 Fort.Vehicle.BonusSpeed {type=Float}
 Fort.Vehicle.EnableShifting {type=Bool}
 Fort.Vehicle.Gear {type=Int32}
 Fort.Vehicle.IsGrounded {type=Bool}
 Fort.Vehicle.Load {type=Float}
 Fort.Vehicle.LowFuel {type=Bool}
 Fort.Vehicle.NormalSpeed {type=Float}
 Fort.Vehicle.OnBlowoff {type=Trigger}
 Fort.Vehicle.OnDownshift {type=Trigger}
 Fort.Vehicle.OnThrottleEngaged {type=Trigger}
 Fort.Vehicle.OnThrottleReleased {type=Trigger}
 Fort.Vehicle.OnUpshift {type=Trigger}
 Fort.Vehicle.Pitch {type=Float}
 Fort.Vehicle.RPM {type=Float}
 Fort.Vehicle.Throttle {type=Float}
 UE.Source.OnPlay {type=Trigger}
 UE.Source.Orientation.Azimuth {type=Float}
 UE.Source.Orientation.Elevation {type=Float}
]
interfaces [Fort.Vehicle:1.0, UE.OutputFormat.Stereo:1.0, UE.Source.OneShot:1.0, UE.Source.Orientation:1.0, UE.Source:1.1]
outputs [
 "UE.OutputFormat.Stereo.Audio:0" {type=Audio}
 "UE.OutputFormat.Stereo.Audio:1" {type=Audio}
 UE.Source.OneShot.OnFinished {type=Trigger}
]
pages [Default {input_defaults {
   Boost.Pitch.Max=2.5
   Boost.Pitch.Min=0
   Boost.PitchByNormalSpeed="<MetaSoundBuilderResult.SUCCEEDED: 0>"
   Boost.Start.Oneshots.DelayTime=0.54
   Boost.Start.Oneshots.NumRepeats=1
   Boost.Start.Oneshots.RepeatSpeed=0.2
   Boost.Start.Oneshots.Volume=0
   Boost.Start.Volume=0
   Fort.Vehicle.Acceleration=0
   Fort.Vehicle.BlowoffStrength=0
   Fort.Vehicle.BonusSpeed=0
   Fort.Vehicle.EnableShifting=null
   Fort.Vehicle.Gear=0
   Fort.Vehicle.IsGrounded="<MetaSoundBuilderResult.SUCCEEDED: 0>"
   Fort.Vehicle.Load=0
   Fort.Vehicle.LowFuel=null
   Fort.Vehicle.NormalSpeed=0.62
   Fort.Vehicle.OnBlowoff=null
   Fort.Vehicle.OnDownshift=null
   Fort.Vehicle.OnThrottleEngaged=null
   Fort.Vehicle.OnThrottleReleased=null
   Fort.Vehicle.OnUpshift=null
   Fort.Vehicle.Pitch=0
   Fort.Vehicle.RPM=0
   Fort.Vehicle.Throttle=0
   UE.Source.OnPlay=null
   UE.Source.Orientation.Azimuth=0
   UE.Source.Orientation.Elevation=0
   Boost.Start.Oneshots.WaveAsset []
   Boost.Start.WaveAsset [&/Boost_Retrostrips_Start_01]
  }}]
properties {
 assetImportData=&self:AssetImportData
 attenuationSettings=/VehicleAudio/Boosts/Vehicle_Boost_Att
 audioPropertiesSheet=/VehicleAudio/PropertySheets/APS_Cosmetic_Vehicle_Boost_Start
 maxDistance=5012
 numChannels=2
 outputFormat=Stereo
 sampleRate=48000
 soundClassObject=/Game/Sounds/FortSoundClassesAndModes/Fort_Vehicle_Exterior_Boost
 soundSubmixObject=&Fortnite_SFX
 concurrencySet [/VehicleAudio/Boosts/Vehicle_Boost_Start_Conc, /VehicleAudio/Boosts/Vehicle_Boost_End_Conc]
 modulationSettings {
  highpassRouting=Union
  lowpassRouting=Union
  volumeRouting=Union
  highpassModulationDestination {bEnableModulation=true, modulators [/Game/Sounds/Modulation/Patches/HPF/PP_FortVehicles_Boost_HPF]}
  lowpassModulationDestination {bEnableModulation=true, modulators [/Game/Sounds/Modulation/Patches/LPF/PP_FortVehicles_Boost_LPF]}
  volumeModulationDestination {
   bEnableModulation=true
   modulators [/VehicleAudio/Modulation/DM_VehicleTurboPatch, /Game/Sounds/Modulation/Patches/PP_FortVehicles_Boost_Vol]
  }
 }
}
